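# Character material for the dark variant of the large male Professor Pup body that scrolls texture coordinates within a masked UV region, as a treadmill-style effect.
# Nested If nodes pan the UVs only where U and V are both below their threshold parameters, with parameterized U and V speeds, leaving other UVs static.
# The selected UVs feed the shared character PreFX function, whose output passes into the PostFX function using material attributes.
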Material /BRCosmetics/Characters/Player/Male/Large/Bodies/M_LRG_Professor_Pup/Materials/Dark/Threadmill_Panner
graph {
 properties {
  bUseMaterialAttributes=true
  bUsedWithMeshDeformer=true
  bUsedWithMorphTargets=true
  bUsedWithSkeletalMesh=true
  pixelDepthOffsetMode=PDOM_Legacy
  usedShadingModels=Default Lit
 }
 nodes [
  n1 = TextureCoordinate()
  n2 = ComponentMask(None=@n1) {g=true}
  n3 = ScalarParameter() {defaultValue=1, group=UVPanner, parameterName=V_Threshold}
  n4 = Panner(
   Coordinate=@n1
   Speed=AppendVector(
    A=ScalarParameter() {group=UVPanner, parameterName=PannerSpeedU}
    B=ScalarParameter() {defaultValue=1, group=UVPanner, parameterName=PannerSpeedV}
   )
  )
  n5 = MaterialFunctionCall(TexCoords=If(
    A=ComponentMask(None=@n1) {r=true}
    B=ScalarParameter() {defaultValue=1, group=UVPanner, parameterName=U_Threshold}
    "A > B"=@n1
    "A == B"=@n1
    "A < B"=If(A=@n2, B=@n3, "A > B"=@n1, "A == B"=@n1, "A < B"=@n4)
   )) {materialFunction=&BaseFN_CharacterFunction_PreFX}
  MaterialFunctionCall(AO=@n5.AO, Distance Blend=@n5.AO, Fresnel=@n5.AO, Material=@n5.AO) {materialFunction=&BaseFN_CharacterFunction_PostFX}
 ]
}
